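# Sound cue for non-looping house ambience of randomized wood creaks, classed as Foundations_Ambience with square map attenuation.
# The first quality branch mixes quiet tree-movement sounds with two delayed layers: very low creaks and pitched-down school creaks, all randomly modulated.
# The second quality branch drops the mixer and delays, randomly choosing from a smaller creak set; the other two branches are empty.
# A concurrency override caps instances at two, blocking new ones.
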SoundCue /Figment_S07/Sounds/Fort_Ambient_NonLoop/Foundation_Mixes/FOUND_RND_Close_Close_House04_Cue
root=QualityLevel_0
nodes [
 {id=Delay_20, children [Modulator_56], properties {delayMax=4, delayMin=3}}
 {id=Delay_21, children [Modulator_1], properties {delayMax=2, delayMin=1}}
 {id=Mixer_16, children [Delay_20,Modulator_57,Delay_21], properties {inputVolume [1,0.15,1]}}
 {id=Modulator_1, children [Random_2], properties {pitchMax=0.8, pitchMin=0.5, volumeMax=1.5, volumeMin=0.3}}
 {id=Modulator_2, children [Random_3], properties {pitchMax=0.8, pitchMin=0.5, volumeMax=1.5, volumeMin=0.3}}
 {id=Modulator_3, children [WavePlayer_9]}
 {id=Modulator_5, children [Mixer_16], properties {pitchMax=1.1, pitchMin=0.9, volumeMax=1.1, volumeMin=0.9}}
 {id=Modulator_56, children [Random_1], properties {pitchMax=1.1, pitchMin=0.9, volumeMax=1.1, volumeMin=0.9}}
 {id=Modulator_57, children [Random_0], properties {pitchMax=1, pitchMin=0.7, volumeMax=1, volumeMin=0.5}}
 {id=Modulator_6, children [Random_4], properties {pitchMax=1.1, pitchMin=0.9, volumeMax=1.1, volumeMin=0.9}}
 {id=QualityLevel_0, children [Modulator_5,Modulator_6,null,null]}
 {id=Random_0, children [WavePlayer_0,WavePlayer_1], properties {weights [1,1]}}
 {id=Random_1, children [WavePlayer_2,WavePlayer_3,WavePlayer_4], properties {weights [1,1,1]}}
 {id=Random_2, children [WavePlayer_11, WavePlayer_14, WavePlayer_7, WavePlayer_8], properties {weights [1,1,1,1]}}
 {id=Random_3, children [WavePlayer_12, WavePlayer_15], properties {weights [1,1]}}
 {id=Random_4, children [Modulator_3,Modulator_2], properties {bRandomizeWithoutReplacement=false, weights [1,1]}}
 {id=WavePlayer_0, properties {soundWaveAssetPtr=/Game/Sounds/Trans_AmbientNonLoop/Wood/TreeMovement_Large01_Stereo}}
 {id=WavePlayer_1, properties {soundWaveAssetPtr=/Game/Sounds/Trans_AmbientNonLoop/Wood/TreeMovement_Large02_Stereo}}
 {id=WavePlayer_11, properties {soundWaveAssetPtr=/Game/Sounds/Fort_Ambient_NonLoop/School/School_LowWoodCreak_01}}
 {id=WavePlayer_12, properties {soundWaveAssetPtr=&School_LowWoodCreak_01}}
 {id=WavePlayer_14, properties {soundWaveAssetPtr=/Game/Sounds/Fort_Ambient_NonLoop/School/School_LowWoodCreak_03}}
 {id=WavePlayer_15, properties {soundWaveAssetPtr=&School_LowWoodCreak_03}}
 {id=WavePlayer_2, properties {soundWaveAssetPtr=/Game/Sounds/Fort_Ambient_NonLoop/Wood/Wood_VeryLow_Creaks_01}}
 {id=WavePlayer_3, properties {soundWaveAssetPtr=/Game/Sounds/Fort_Ambient_NonLoop/Wood/Wood_VeryLow_Creaks_02}}
 {id=WavePlayer_4, properties {soundWaveAssetPtr=/Game/Sounds/Fort_Ambient_NonLoop/Wood/Wood_VeryLow_Creaks_03}}
 {id=WavePlayer_7, properties {soundWaveAssetPtr=/Game/Sounds/Fort_Ambient_NonLoop/School/School_LowWoodCreak_02}}
 {id=WavePlayer_8, properties {soundWaveAssetPtr=/Game/Sounds/Fort_Ambient_NonLoop/School/School_LowWoodCreak_04}}
 {id=WavePlayer_9, properties {soundWaveAssetPtr=&Wood_VeryLow_Creaks_02}}
]
properties {
 attenuationSettings=/Game/Sounds/Attenuation/Fort_Map_3x3_Square
 bOverrideConcurrency=true
 maxDistance=1780
 soundClassObject=/Figment_S07/Sounds/FortSoundClassesAndModes/Foundations_Ambience
 volumeMultiplier=0.6
 concurrencyOverrides {maxCount=2, resolutionRule=PreventNew}
}
